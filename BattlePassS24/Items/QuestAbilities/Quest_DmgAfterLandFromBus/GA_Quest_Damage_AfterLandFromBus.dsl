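; Gameplay ability for a Season 24 Battle Pass quest, named for dealing damage after landing from the bus.
; On event-triggered activation it calls the parent, applies a gameplay effect to the owning player pawn's ability system component, then ends.
; The effect class comes from a variable named GE_Quest_AfterMantling5s, whose mantling name differs from the landing name of the ability.
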(event Ability|EventActivateAbilityFromEvent
  (params (EventData :type Gameplay%20Event%20Data%20Structure%20%28by%20ref%29))
  (|CallParentFunction EventData)
  (GameplayEffects|ApplyGameplayEffectToSelf :self (Ability|GetFortAbilitySystemComponent (Variables|GAVars|Data|GetOwningPlayerPawn)) :GameplayEffectClass (Variables|Default|GetGE_Quest_AfterMantling5s) :Level 0.0)
  (Ability|EndAbility))
